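# Physics asset for the Ravenquill backpack cosmetic, simulating the left wing's dynamic bones.
# Four zero-length capsules, effectively spheres, form a chain from backpack root through wing base and mid to bottom.
# The base joint allows only a narrow swing on one axis, and the mid joint only twist.
# The file defines no right-wing bodies or constraints.
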PhysicsAsset /BRCosmetics/Accessories/FORT_Backpacks/F_MED_Ravenquill/Export/Backpack_F_RavenQuill_Physics
bodies {
 backpack_root [{shapeType=Capsule, length=0, radius=2.5}]
 dyn_wing_base_l [{shapeType=Capsule, length=0, radius=1.25}]
 dyn_wing_bottom_l [{shapeType=Capsule, length=0, radius=5}]
 dyn_wing_mid_l [{shapeType=Capsule, length=0, radius=1}]
}
constraints [
 {bone1Name=dyn_wing_bottom_l, bone2Name=dyn_wing_mid_l, swing1LimitDegrees=0, swing2LimitDegrees=0, twistLimitDegrees=12.5}
 {
  bone1Name=dyn_wing_mid_l
  bone2Name=dyn_wing_base_l
  swing1LimitDegrees=0
  swing1Motion=Locked
  swing2LimitDegrees=0
  swing2Motion=Locked
  twistLimitDegrees=12.5
 }
 {
  bone1Name=dyn_wing_base_l
  bone2Name=backpack_root
  swing1LimitDegrees=10
  swing2LimitDegrees=0
  swing2Motion=Locked
  twistLimitDegrees=0
  twistMotion=Locked
 }
]
